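# Editor utility widget that builds Fort Player foley wave container assets from Sound Waves in the current Content Browser folder.
# Generate scans that folder non-recursively, and SetTokens reads type, surface and action from underscore-split names.
# Only an action change saves collected waves into the previous container's WaveAssets and creates DT_Foley_Waves_Type_Action_Surface under DataAssets, so grouping depends on asset order.
# Loop completion fills the last container.
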EditorUtilityWidgetBlueprint /Game/Sounds/Fort_Player/Chooser/Tools/EUW_FoleyWaveAssetContainerUtil
graphs [
 EventGraph {nodes [
   "On Clicked (Generate)"() {is=title K2Node_ComponentBoundEvent}
   > "SetText (Text Box)"!(
    execute
    self:Text Box Object Reference=Get ScanDirectory() -> (ScanDirectory:Editor Utility Editable Text Box Object Reference)
    InText:Text="To Text (String)"(InString=GetCurrentContentBrowserPath!() -> (OutPath:String, ReturnValue:bool))
   ) -> (then)
   > Set CurrentActionToken(CurrentActionToken:String) -> (Output_Get:String)
   > Set CurrentSurfaceToken(CurrentSurfaceToken:String) -> (Output_Get:String)
   > Set CurrentTypeToken(CurrentTypeToken:String) -> (Output_Get:String)
   > Set PreviousActionToken(PreviousActionToken:String) -> (Output_Get:String)
   n1 = Get Text(self=Get ScanDirectory())
   > n2 = GetAssetsByPath!(
    execute
    self:Asset Registry Interface=GetAssetRegistry()
    PackagePath:Name=String To Name(InString="To String (Text)"(InText=@n1))
    bRecursive:bool=false
    bIncludeOnlyOnDiskAssets:bool=false
   ) -> (then, OutAssetData:Array of Asset Data Structures, ReturnValue:bool)
   > n3 = For Each Loop(Array:Array of Asset Data Structures=@n2) -> (Array Element:Asset Data Structure)
   n4 = Get!(execute:Exec=@n3.Completed) -> (then:Exec, else:Exec, CurrentDataAsset:Fort Player Foley Wave Asset Container Object Reference)
   > SetEditorProperty(
    Object=@n4
    PropertyName=WaveAssets
    PropertyValue:Array of Sound Wave Object References=Get CurrentWaveAssets() -> (CurrentWaveAssets:Array of Sound Wave Object References)
   )
   Set AssetData(execute=@n3.LoopBody, AssetData:Asset Data Structure=@n3) -> (Output_Get:Asset Data Structure)
   > n5 = Cast To SoundWave!(execute, Object=GetAsset(InAssetData=Get AssetData())) -> (then, CastFailed, AsSound Wave:Sound Wave Object Reference) {is=title K2Node_DynamicCast}
   > n6 = Set LoadedAsset(LoadedAsset:Sound Wave Object Reference=@n5) -> (Output_Get:Sound Wave Object Reference)
   n7 = "To String (Name)"(InName=Break Asset Data(AssetData=Get AssetData()).AssetName)
   > Set AssetName(AssetName:String=@n7) -> (Output_Get:String)
   n8 = ParseIntoArray(SourceString=@n7, Delimiter=_)
   > Set Tokens(Tokens:Array of Strings=@n8) -> (Output_Get:Array of Strings)
   > Set TokenCount(TokenCount:int=Length(TargetArray:Array of Strings=@n8)) -> (Output_Get:int=0)
   > SetTokens()
   > PrintString(InString=Append!(
     A=GetDisplayName(Object=@n6)
     B=", "
     C:String=Get CurrentActionToken() -> (CurrentActionToken:String)
     D:String=", "
     E:String=Get CurrentSurfaceToken() -> (CurrentSurfaceToken:String)
    ) -> (ReturnValue))
   Event PreConstruct()
   Event Construct()
   Event Tick()
  ]}
 SetTokens {nodes [
   SetTokens() {is=title K2Node_FunctionEntry}
   > n1 = Sequence!(execute) -> (then_0, then_1, then_2:Exec, then_3:Exec)
   > Branch(Condition="Equal Exactly (String)"(A=Get CurrentActionToken(), B=Walk))
   > PrintString(InString=Hello)
   Set PreviousActionToken(execute=@n1.then_1, PreviousActionToken=Get CurrentActionToken())
   n2 = Get Tokens() -> (Tokens:Array of Strings)
   > Set CurrentActionToken(CurrentActionToken="Get (a copy)"(Array=@n2, Dimension 1=5))
   > Set CurrentSurfaceToken(CurrentSurfaceToken="Get (a copy)"(Array=@n2, Dimension 1=4))
   > Set CurrentTypeToken(CurrentTypeToken="Get (a copy)"(Array=@n2, Dimension 1=3))
   Branch(
    execute=@n1.then_2
    Condition="Not Equal Exactly (String)"(A=Get CurrentActionToken(), B=Get PreviousActionToken() -> (PreviousActionToken:String))
   )
   > n3 = Sequence!(execute) -> (then_0, then_1)
   > n4 = Get()
   > SetEditorProperty(Object=@n4, PropertyName=WaveAssets, PropertyValue=Get CurrentWaveAssets())
   n5 = GetAssetTools(execute=@n3.then_1)
   n6 = Get CurrentActionToken()
   n7 = Get CurrentSurfaceToken()
   n8 = Get CurrentTypeToken() -> (CurrentTypeToken:String)
   n9 = Get Text(self=Get ScanDirectory())
   > n10 = CreateAsset(
    self=@n5
    AssetName="To String (Text)"(InText=Format Text!(Format="LOCTEXT(\"DT_Foley_Waves_{Type}_{Action}_{Surface}\")", Type:String=@n8, Action:String=@n6, Surface:String=@n7) -> (Result))
    PackagePath=Append!(A="To String (Text)"(InText=@n9), B=/DataAssets) -> (ReturnValue)
    AssetClass=/Script/FortniteGame.FortPlayerFoleyWaveAssetContainer
    CallingContext=Foley
   )
   > n11 = Cast To FortPlayerFoleyWaveAssetContainer!(execute, Object=@n10) -> (then, CastFailed, AsFort Player Foley Wave Asset Container:Fort Player Foley Wave Asset Container Object Reference) {is=title K2Node_DynamicCast}
   > Set CurrentDataAsset(CurrentDataAsset:Fort Player Foley Wave Asset Container Object Reference=@n11) -> (Output_Get:Fort Player Foley Wave Asset Container Object Reference)
   > Clear(TargetArray:Array of Sound Wave Object References=Get CurrentWaveAssets())
   Add(
    execute=@n1.then_3
    TargetArray:Array of Sound Wave Object References=Get CurrentWaveAssets()
    NewItem:"Sound Wave Object Reference (by ref)"=Get LoadedAsset() -> (LoadedAsset:Sound Wave Object Reference)
   ) {is=title K2Node_CallArrayFunction}
  ]}
]
variables [
 AssetData {type=AssetData, default {}}
 AssetName {type=string, default=""}
 Blah {type="array<MySoundWave>", default []}
 CurrentActionToken {type=string, default=""}
 CurrentDataAsset {type=FortPlayerFoleyWaveAssetContainer, default=null}
 CurrentSurfaceToken {type=string, default=""}
 CurrentTypeToken {type=string, default=""}
 CurrentWaveAssets {type="array<SoundWave>", default []}
 LoadedAsset {type=SoundWave, default=null}
 PreviousActionToken {type=string, default=""}
 TokenCount {type=integer}
 Tokens {type="array<string>", default []}
]
widgets {
 info {rootWidgetClass=&CanvasPanel, widgetCount=7}
 widgets [
  CanvasPanel_0 {widgetClassPath=&CanvasPanel, widget=CanvasPanel_0}
  StackBox_1 {
   widgetClassPath=&StackBox
   parent=CanvasPanel_0
   widget=StackBox_1
   properties {orientation=Orient_Vertical}
   slot {class=CanvasPanelSlot, properties {layoutData {alignment=<0 0>, anchors {maximum=<0 0>, minimum=<0 0>}, offsets=<ltrb 0 0 400 150>}}}
  }
  StackBox_0 {widgetClassPath=&StackBox, parent=StackBox_1, widget=StackBox_0, slot {class=StackBoxSlot}}
  ScanDirectory {
   widgetClassPath=&EditorUtilityEditableTextBox
   bIsVariable=true
   parent=StackBox_0
   widget=ScanDirectory
   properties {widgetStyle {
     backgroundColor {colorUseRule=UseColor_Specified, specifiedColor=<rgba 1 1 1 1>}
     backgroundImageFocused {
      bIsDynamicallyLoaded=false
      drawAs=RoundedBox
      imageType=NoImage
      mirroring=NoMirror
      resourceName=None
      resourceObject=null
      tiling=NoTile
      imageSize=<0 0>
      margin=<ltrb 0 0 0 0>
      outlineSettings {
       bUseBrushTransparency=false
       roundingType=FixedRadius
       width=1
       color {colorUseRule=UseColor_ColorTable, specifiedColor=<rgba 0 0 0 0>}
       cornerRadii=<4 4 4 4>
      }
      tintColor {colorUseRule=UseColor_Specified, specifiedColor=<rgba 0.0048 0.0048 0.0048 1>}
      uVRegion {bIsValid=false, max=<0 0>, min=<0 0>}
     }
     backgroundImageHovered {
      bIsDynamicallyLoaded=false
      drawAs=RoundedBox
      imageType=NoImage
      mirroring=NoMirror
      resourceName=None
      resourceObject=null
      tiling=NoTile
      imageSize=<0 0>
      margin=<ltrb 0 0 0 0>
      outlineSettings {
       bUseBrushTransparency=false
       roundingType=FixedRadius
       width=1
       color {colorUseRule=UseColor_ColorTable, specifiedColor=<rgba 0 0 0 0>}
       cornerRadii=<4 4 4 4>
      }
      tintColor {colorUseRule=UseColor_Specified, specifiedColor=<rgba 0.0048 0.0048 0.0048 1>}
      uVRegion {bIsValid=false, max=<0 0>, min=<0 0>}
     }
     backgroundImageNormal {
      bIsDynamicallyLoaded=false
      drawAs=RoundedBox
      imageType=NoImage
      mirroring=NoMirror
      resourceName=None
      resourceObject=null
      tiling=NoTile
      imageSize=<0 0>
      margin=<ltrb 0 0 0 0>
      outlineSettings {
       bUseBrushTransparency=false
       roundingType=FixedRadius
       width=1
       color {colorUseRule=UseColor_ColorTable, specifiedColor=<rgba 0 0 0 0>}
       cornerRadii=<4 4 4 4>
      }
      tintColor {colorUseRule=UseColor_Specified, specifiedColor=<rgba 0.0048 0.0048 0.0048 1>}
      uVRegion {bIsValid=false, max=<0 0>, min=<0 0>}
     }
     backgroundImageReadOnly {
      bIsDynamicallyLoaded=false
      drawAs=RoundedBox
      imageType=NoImage
      mirroring=NoMirror
      resourceName=None
      resourceObject=null
      tiling=NoTile
      imageSize=<0 0>
      margin=<ltrb 0 0 0 0>
      outlineSettings {
       bUseBrushTransparency=false
       roundingType=FixedRadius
       width=1
       color {colorUseRule=UseColor_ColorTable, specifiedColor=<rgba 0 0 0 0>}
       cornerRadii=<4 4 4 4>
      }
      tintColor {colorUseRule=UseColor_Specified, specifiedColor=<rgba 0.0284 0.0284 0.0284 1>}
      uVRegion {bIsValid=false, max=<0 0>, min=<0 0>}
     }
     focusedForegroundColor {colorUseRule=UseColor_Specified, specifiedColor=<rgba 1 1 1 1>}
     foregroundColor {colorUseRule=UseColor_Specified, specifiedColor=<rgba 0.5271 0.5271 0.5271 1>}
     hScrollBarPadding=<ltrb 0 0 0 0>
     padding=<ltrb 8 3 8 4>
     readOnlyForegroundColor {colorUseRule=UseColor_Foreground, specifiedColor=<rgba 1 0 1 1>}
     scrollBarStyle {
      thickness=8
      draggedThumbImage {
       bIsDynamicallyLoaded=false
       drawAs=RoundedBox
       imageType=NoImage
       mirroring=NoMirror
       resourceName=None
       resourceObject=null
       tiling=NoTile
       imageSize=<0 0>
       margin=<ltrb 0 0 0 0>
       outlineSettings {
        bUseBrushTransparency=false
        roundingType=FixedRadius
        width=0
        color {colorUseRule=UseColor_Specified, specifiedColor=<rgba 0 0 0 0>}
        cornerRadii=<4 4 4 4>
       }
       tintColor {colorUseRule=UseColor_Specified, specifiedColor=<rgba 0.2159 0.2159 0.2159 1>}
       uVRegion {bIsValid=false, max=<0 0>, min=<0 0>}
      }
      horizontalBackgroundImage {
       bIsDynamicallyLoaded=false
       drawAs=NoDrawType
       imageType=NoImage
       mirroring=NoMirror
       resourceName=None
       resourceObject=null
       tiling=NoTile
       imageSize=<0 0>
       margin=<ltrb 0 0 0 0>
       outlineSettings {
        bUseBrushTransparency=false
        roundingType=HalfHeightRadius
        width=0
        color {colorUseRule=UseColor_Specified, specifiedColor=<rgba 0 0 0 0>}
        cornerRadii=<0 0 0 0>
       }
       tintColor {colorUseRule=UseColor_Specified, specifiedColor=<rgba 1 1 1 1>}
       uVRegion {bIsValid=false, max=<0 0>, min=<0 0>}
      }
      horizontalBottomSlotImage {
       bIsDynamicallyLoaded=false
       drawAs=NoDrawType
       imageType=NoImage
       mirroring=NoMirror
       resourceName=None
       resourceObject=null
       tiling=NoTile
       imageSize=<0 0>
       margin=<ltrb 0 0 0 0>
       outlineSettings {
        bUseBrushTransparency=false
        roundingType=HalfHeightRadius
        width=0
        color {colorUseRule=UseColor_Specified, specifiedColor=<rgba 0 0 0 0>}
        cornerRadii=<0 0 0 0>
       }
       tintColor {colorUseRule=UseColor_Specified, specifiedColor=<rgba 1 1 1 1>}
       uVRegion {bIsValid=false, max=<0 0>, min=<0 0>}
      }
      horizontalTopSlotImage {
       bIsDynamicallyLoaded=false
       drawAs=NoDrawType
       imageType=NoImage
       mirroring=NoMirror
       resourceName=None
       resourceObject=null
       tiling=NoTile
       imageSize=<0 0>
       margin=<ltrb 0 0 0 0>
       outlineSettings {
        bUseBrushTransparency=false
        roundingType=HalfHeightRadius
        width=0
        color {colorUseRule=UseColor_Specified, specifiedColor=<rgba 0 0 0 0>}
        cornerRadii=<0 0 0 0>
       }
       tintColor {colorUseRule=UseColor_Specified, specifiedColor=<rgba 1 1 1 1>}
       uVRegion {bIsValid=false, max=<0 0>, min=<0 0>}
      }
      hoveredThumbImage {
       bIsDynamicallyLoaded=false
       drawAs=RoundedBox
       imageType=NoImage
       mirroring=NoMirror
       resourceName=None
       resourceObject=null
       tiling=NoTile
       imageSize=<0 0>
       margin=<ltrb 0 0 0 0>
       outlineSettings {
        bUseBrushTransparency=false
        roundingType=FixedRadius
        width=0
        color {colorUseRule=UseColor_Specified, specifiedColor=<rgba 0 0 0 0>}
        cornerRadii=<4 4 4 4>
       }
       tintColor {colorUseRule=UseColor_Specified, specifiedColor=<rgba 0.2159 0.2159 0.2159 1>}
       uVRegion {bIsValid=false, max=<0 0>, min=<0 0>}
      }
      normalThumbImage {
       bIsDynamicallyLoaded=false
       drawAs=RoundedBox
       imageType=NoImage
       mirroring=NoMirror
       resourceName=None
       resourceObject=null
       tiling=NoTile
       imageSize=<0 0>
       margin=<ltrb 0 0 0 0>
       outlineSettings {
        bUseBrushTransparency=false
        roundingType=FixedRadius
        width=0
        color {colorUseRule=UseColor_Specified, specifiedColor=<rgba 0 0 0 0>}
        cornerRadii=<4 4 4 4>
       }
       tintColor {colorUseRule=UseColor_Specified, specifiedColor=<rgba 0.0953 0.0953 0.0953 1>}
       uVRegion {bIsValid=false, max=<0 0>, min=<0 0>}
      }
      verticalBackgroundImage {
       bIsDynamicallyLoaded=false
       drawAs=NoDrawType
       imageType=NoImage
       mirroring=NoMirror
       resourceName=None
       resourceObject=null
       tiling=NoTile
       imageSize=<0 0>
       margin=<ltrb 0 0 0 0>
       outlineSettings {
        bUseBrushTransparency=false
        roundingType=HalfHeightRadius
        width=0
        color {colorUseRule=UseColor_Specified, specifiedColor=<rgba 0 0 0 0>}
        cornerRadii=<0 0 0 0>
       }
       tintColor {colorUseRule=UseColor_Specified, specifiedColor=<rgba 1 1 1 1>}
       uVRegion {bIsValid=false, max=<0 0>, min=<0 0>}
      }
      verticalBottomSlotImage {
       bIsDynamicallyLoaded=false
       drawAs=NoDrawType
       imageType=NoImage
       mirroring=NoMirror
       resourceName=None
       resourceObject=null
       tiling=NoTile
       imageSize=<0 0>
       margin=<ltrb 0 0 0 0>
       outlineSettings {
        bUseBrushTransparency=false
        roundingType=HalfHeightRadius
        width=0
        color {colorUseRule=UseColor_Specified, specifiedColor=<rgba 0 0 0 0>}
        cornerRadii=<0 0 0 0>
       }
       tintColor {colorUseRule=UseColor_Specified, specifiedColor=<rgba 1 1 1 1>}
       uVRegion {bIsValid=false, max=<0 0>, min=<0 0>}
      }
      verticalTopSlotImage {
       bIsDynamicallyLoaded=false
       drawAs=NoDrawType
       imageType=NoImage
       mirroring=NoMirror
       resourceName=None
       resourceObject=null
       tiling=NoTile
       imageSize=<0 0>
       margin=<ltrb 0 0 0 0>
       outlineSettings {
        bUseBrushTransparency=false
        roundingType=HalfHeightRadius
        width=0
        color {colorUseRule=UseColor_Specified, specifiedColor=<rgba 0 0 0 0>}
        cornerRadii=<0 0 0 0>
       }
       tintColor {colorUseRule=UseColor_Specified, specifiedColor=<rgba 1 1 1 1>}
       uVRegion {bIsValid=false, max=<0 0>, min=<0 0>}
      }
     }
     textStyle {
      overflowPolicy=Clip
      transformPolicy=None
      colorAndOpacity {colorUseRule=UseColor_Foreground, specifiedColor=<rgba 1 0 1 1>}
      font {
       bForceMonospaced=false
       bMaterialIsStencil=false
       fontMaterial=null
       fontObject=null
       letterSpacing=0
       monospacedWidth=1
       size=15
       skewAmount=0
       typefaceFontName=Regular
       outlineSettings {
        bApplyOutlineToDropShadows=false
        bMiteredCorners=false
        bSeparateFillAlpha=false
        outlineMaterial=null
        outlineSize=0
        outlineColor=<rgba 0 0 0 1>
       }
      }
      highlightColor {colorUseRule=UseColor_ColorTable, specifiedColor=<rgba 0 0 0 0>}
      highlightShape {
       bIsDynamicallyLoaded=false
       drawAs=Image
       imageType=NoImage
       mirroring=NoMirror
       resourceName=None
       resourceObject=null
       tiling=NoTile
       imageSize=<0 0>
       margin=<ltrb 0 0 0 0>
       outlineSettings {
        bUseBrushTransparency=false
        roundingType=HalfHeightRadius
        width=0
        color {colorUseRule=UseColor_Specified, specifiedColor=<rgba 0 0 0 0>}
        cornerRadii=<0 0 0 0>
       }
       tintColor {colorUseRule=UseColor_ColorTable, specifiedColor=<rgba 0 0 0 0>}
       uVRegion {bIsValid=false, max=<0 0>, min=<0 0>}
      }
      selectedBackgroundColor {colorUseRule=UseColor_ColorTable, specifiedColor=<rgba 0 0 0 0>}
      shadowColorAndOpacity=<rgba 0 0 0 1>
      shadowOffset=<0 0>
      strikeBrush {
       bIsDynamicallyLoaded=false
       drawAs=Image
       imageType=NoImage
       mirroring=NoMirror
       resourceName=None
       resourceObject=null
       tiling=NoTile
       imageSize=<32 32>
       margin=<ltrb 0 0 0 0>
       outlineSettings {
        bUseBrushTransparency=false
        roundingType=HalfHeightRadius
        width=0
        color {colorUseRule=UseColor_Specified, specifiedColor=<rgba 0 0 0 0>}
        cornerRadii=<0 0 0 0>
       }
       tintColor {colorUseRule=UseColor_Specified, specifiedColor=<rgba 1 1 1 1>}
       uVRegion {bIsValid=false, max=<0 0>, min=<0 0>}
      }
      underlineBrush {
       bIsDynamicallyLoaded=false
       drawAs=Image
       imageType=NoImage
       mirroring=NoMirror
       resourceName=None
       resourceObject=null
       tiling=NoTile
       imageSize=<32 32>
       margin=<ltrb 0 0 0 0>
       outlineSettings {
        bUseBrushTransparency=false
        roundingType=HalfHeightRadius
        width=0
        color {colorUseRule=UseColor_Specified, specifiedColor=<rgba 0 0 0 0>}
        cornerRadii=<0 0 0 0>
       }
       tintColor {colorUseRule=UseColor_Specified, specifiedColor=<rgba 1 1 1 1>}
       uVRegion {bIsValid=false, max=<0 0>, min=<0 0>}
      }
     }
     vScrollBarPadding=<ltrb 0 0 0 0>
    }}
   slot {class=StackBoxSlot, properties {size {sizeRule=Fill, value=1}}}
  }
  StackBox {widgetClassPath=&StackBox, parent=StackBox_1, widget=StackBox, slot {class=StackBoxSlot}}
  Generate {
   widgetClassPath=&EditorUtilityButton
   parent=StackBox
   widget=Generate
   slot {class=StackBoxSlot, properties {verticalAlignment=VAlign_Top, size {sizeRule=Fill, value=1}}}
  }
  TextBlock_65 {widgetClassPath=&TextBlock, parent=Generate, widget=TextBlock_65, properties {text=Generate}, slot {class=ButtonSlot}}
 ]
}
